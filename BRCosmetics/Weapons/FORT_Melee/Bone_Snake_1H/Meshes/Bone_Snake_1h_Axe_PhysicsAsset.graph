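# Physics asset for the Bone Snake one-handed axe, a melee weapon cosmetic.
# It has two small sphere bodies, one on the b_AR_Root bone and one on the dyn_cloth bone.
# A single constraint links the pair with wide swing limits and a tight twist limit, so the dyn_cloth part can swing relative to the root with little twist.
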PhysicsAsset /BRCosmetics/Weapons/FORT_Melee/Bone_Snake_1H/Meshes/Bone_Snake_1h_Axe_PhysicsAsset
bodies {
 b_AR_Root [{shapeType=Sphere, radius=4.8549, center=<2.2534 10.7066 -2.8577>}]
 dyn_cloth [{shapeType=Sphere, radius=4.8556, center=<4.7812 0.4398 0>}]
}
constraints [{bone1Name=dyn_cloth, bone2Name=b_AR_Root, swing1LimitDegrees=75, swing2LimitDegrees=75, twistLimitDegrees=5}]
